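# Anim Blueprint for the ClumsyChew pickaxe that adds physics sway to its three-bone tail.
# Two chained Anim Dynamics sets serve gameplay (bone-relative) and front end, each zeroing the other's alphas.
# The front-end set simulates in world space with downward gravity, switching to bone-relative during swings.
# Sliding drives the first tail's gameplay alpha, also enabling an additive yaw after that bone's translation is constrained to the tail base.
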AnimBlueprint /BRCosmetics/Weapons/FORT_Melee/Pickaxe_ClumsyChew/Meshes/ClumsyChew_Axe_AnimBP
target_skeleton=&/ClumsyChew_Axe_Skeleton
graphs [
 AnimGraph {nodes [
   AnimGraphNode_SubInput_0 = Input Pose()
   n1 = Anim Dynamics - Bone: dyn_sword_tail_1(
    ComponentPose=Local To Component(LocalPose=@AnimGraphNode_SubInput_0)
    Alpha=Get Dyn_Tail_1_Alpha() -> (Dyn_Tail_1_Alpha:float)
    ExternalForce="0,0,0"
   ) {
    angularBiasOverride=0.9
    angularDampingOverride=1
    angularSpringConstant=20
    bAngularSpring=true
    bGravityOverrideInSimSpace=true
    bOverrideAngularBias=true
    bOverrideAngularDamping=true
    bUseGravityOverride=true
    linearDampingOverride=0.9
    simulationSpace=BoneRelative
    alphaBoolBlend {blendOption=Linear}
    boundBone {boneName=dyn_sword_tail_1}
    componentAppliedLinearAccClamp=<600 600 600>
    componentLinearAccScale=<0.75 0.75 0.75>
    componentLinearVelScale=<0.75 0.75 0.75>
    gravityOverride {x=-1900}
    physicsBodyDefinitions [{
      boundBone {boneName=dyn_sword_tail_1}
      boxExtents=<60 25 25>
      constraintSetup {angularLimitsMax=<5 75 75>, angularLimitsMin=<-5 -75 -75>, angularTarget=<0 0 0>}
      localJointOffset=<40 0 0>
     }]
    relativeSpaceBone {boneName=b_AR_Root}
   }
   n2 = Anim Dynamics - Bone: dyn_sword_tail_2(ComponentPose=@n1, Alpha=Get Dyn_Tail_2_Alpha() -> (Dyn_Tail_2_Alpha:float), ExternalForce="0,0,0") {
    angularBiasOverride=0.9
    angularDampingOverride=1
    angularSpringConstant=15
    bAngularSpring=true
    bGravityOverrideInSimSpace=true
    bOverrideAngularBias=true
    bOverrideAngularDamping=true
    bUseGravityOverride=true
    linearDampingOverride=0.9
    simulationSpace=BoneRelative
    alphaBoolBlend {blendOption=Linear}
    boundBone {boneName=dyn_sword_tail_2}
    componentAppliedLinearAccClamp=<900 900 900>
    componentLinearAccScale=<0.75 0.75 0.75>
    componentLinearVelScale=<0.75 0.75 0.75>
    gravityOverride {x=-1250}
    physicsBodyDefinitions [{
      boundBone {boneName=dyn_sword_tail_2}
      boxExtents=<20 15 15>
      constraintSetup {angularLimitsMax=<5 45 45>, angularLimitsMin=<-5 -45 -45>, angularTarget=<0 0 0>}
      localJointOffset=<15 0 0>
     }]
    relativeSpaceBone {boneName=b_AR_Root}
   }
   n3 = Anim Dynamics - Bone: dyn_sword_tail_3(ComponentPose=@n2, Alpha=Get Dyn_Tail_3_Alpha() -> (Dyn_Tail_3_Alpha:float), ExternalForce="0,0,0") {
    angularBiasOverride=0.9
    angularDampingOverride=1
    angularSpringConstant=7.5
    bAngularSpring=true
    bGravityOverrideInSimSpace=true
    bOverrideAngularBias=true
    bOverrideAngularDamping=true
    bUseGravityOverride=true
    linearDampingOverride=0.9
    simulationSpace=BoneRelative
    alphaBoolBlend {blendOption=Linear}
    boundBone {boneName=dyn_sword_tail_3}
    componentAppliedLinearAccClamp=<1000 1000 1000>
    componentLinearAccScale=<0.75 0.75 0.75>
    componentLinearVelScale=<0.75 0.75 0.75>
    gravityOverride {x=-980}
    physicsBodyDefinitions [{
      boundBone {boneName=dyn_sword_tail_3}
      boxExtents=<15 7.5 7.5>
      constraintSetup {angularLimitsMax=<20 60 60>, angularLimitsMin=<-20 -60 -60>, angularTarget=<0 0 0>}
      localJointOffset=<10 0 0>
     }]
    relativeSpaceBone {boneName=b_AR_Root}
   }
   n4 = Get Dyn_Simulation_Space() -> (Dyn_Simulation_Space:AnimPhysSimSpaceType Enum)
   n5 = Anim Dynamics - Bone: dyn_sword_tail_1!(
    SimulationSpace:AnimPhysSimSpaceType Enum=@n4
    ComponentPose=@n3
    Alpha=Get Dyn_Tail_1_Alpha_FE() -> (Dyn_Tail_1_Alpha_FE:float)
    GravityOverride:Vector=Get Dyn Gravity Override 1() -> (Dyn Gravity Override 1:Vector)
    ExternalForce="0,0,0"
   ) -> (Pose) {
    angularBiasOverride=0.9
    angularDampingOverride=1
    angularSpringConstant=20
    bAngularSpring=true
    bGravityOverrideInSimSpace=true
    bOverrideAngularBias=true
    bOverrideAngularDamping=true
    bUseGravityOverride=true
    linearDampingOverride=0.9
    simulationSpace=World
    alphaBoolBlend {blendOption=Linear}
    boundBone {boneName=dyn_sword_tail_1}
    componentAppliedLinearAccClamp=<300 300 300>
    componentLinearAccScale=<0.75 0.75 0.75>
    componentLinearVelScale=<0.75 0.75 0.75>
    gravityOverride {x=-1900}
    physicsBodyDefinitions [{
      boundBone {boneName=dyn_sword_tail_1}
      boxExtents=<60 25 25>
      constraintSetup {angularLimitsMax=<5 75 75>, angularLimitsMin=<-5 -75 -75>, angularTarget=<0 0 0>}
      localJointOffset=<40 0 0>
     }]
    relativeSpaceBone {boneName=b_AR_Root}
   }
   n6 = Anim Dynamics - Bone: dyn_sword_tail_2!(
    SimulationSpace:AnimPhysSimSpaceType Enum=@n4
    ComponentPose=@n5
    Alpha=Get Dyn_Tail_2_Alpha_FE() -> (Dyn_Tail_2_Alpha_FE:float)
    GravityOverride:Vector=Get Dyn Gravity Override 2() -> (Dyn Gravity Override 2:Vector)
    ExternalForce="0,0,0"
   ) -> (Pose) {
    angularBiasOverride=0.9
    angularDampingOverride=1
    angularSpringConstant=15
    bAngularSpring=true
    bGravityOverrideInSimSpace=true
    bOverrideAngularBias=true
    bOverrideAngularDamping=true
    bUseGravityOverride=true
    linearDampingOverride=0.9
    simulationSpace=World
    alphaBoolBlend {blendOption=Linear}
    boundBone {boneName=dyn_sword_tail_2}
    componentAppliedLinearAccClamp=<450 450 450>
    componentLinearAccScale=<0.75 0.75 0.75>
    componentLinearVelScale=<0.75 0.75 0.75>
    gravityOverride {x=-1250}
    physicsBodyDefinitions [{
      boundBone {boneName=dyn_sword_tail_2}
      boxExtents=<20 15 15>
      constraintSetup {angularLimitsMax=<5 45 45>, angularLimitsMin=<-5 -45 -45>, angularTarget=<0 0 0>}
      localJointOffset=<15 0 0>
     }]
    relativeSpaceBone {boneName=b_AR_Root}
   }
   n7 = Anim Dynamics - Bone: dyn_sword_tail_3!(
    SimulationSpace:AnimPhysSimSpaceType Enum=@n4
    ComponentPose=@n6
    Alpha=Get Dyn_Tail_3_Alpha_FE() -> (Dyn_Tail_3_Alpha_FE:float)
    GravityOverride:Vector=Get Dyn Gravity Override 3() -> (Dyn Gravity Override 3:Vector)
    ExternalForce="0,0,0"
   ) -> (Pose) {
    angularBiasOverride=0.9
    angularDampingOverride=1
    angularSpringConstant=7.5
    bAngularSpring=true
    bGravityOverrideInSimSpace=true
    bOverrideAngularBias=true
    bOverrideAngularDamping=true
    bUseGravityOverride=true
    linearDampingOverride=0.9
    simulationSpace=World
    alphaBoolBlend {blendOption=Linear}
    boundBone {boneName=dyn_sword_tail_3}
    componentAppliedLinearAccClamp=<600 600 600>
    componentLinearAccScale=<0.75 0.75 0.75>
    componentLinearVelScale=<0.75 0.75 0.75>
    gravityOverride {x=-980}
    physicsBodyDefinitions [{
      boundBone {boneName=dyn_sword_tail_3}
      boxExtents=<15 7.5 7.5>
      constraintSetup {angularLimitsMax=<20 60 60>, angularLimitsMin=<-20 -60 -60>, angularTarget=<0 0 0>}
      localJointOffset=<10 0 0>
     }]
    relativeSpaceBone {boneName=b_AR_Root}
   }
   n8 = "Constraint - dyn_sword_tail_1 (T)"(ComponentPose=@n7, Alpha=1, ConstraintWeights_0=1) {
    alphaBoolBlend {blendOption=Linear}
    boneToModify {boneName=dyn_sword_tail_1}
    constraintSetup [{transformType=Translation, targetBone {boneName=dyn_sword_tail_base}}]
    constraintWeights [1]
   }
   n9 = "Transform (Modify) Bone - Bone: dyn_sword_tail_1"(ComponentPose=@n8, Alpha=Get Dyn_Tail_1_Alpha(), Translation="0,0,0", Rotation="0,45,0", Scale="1,1,1") {
    rotationMode=BMM_Additive
    rotationSpace=BCS_BoneSpace
    alphaBoolBlend {blendOption=Linear}
    boneToModify {boneName=dyn_sword_tail_1}
    rotation {yaw=45}
   }
   Output Pose(Result=Component To Local(ComponentPose=@n9)) {layerGroup=None}
  ]}
 EventGraph {nodes [
   Event BlueprintUpdateAnimation()
   > n1 = Sequence()
   > n2 = Branch(Condition=Get bIsInFrontEnd())
   > n3 = Branch(Condition=Get Dyn_Pickaxe_Swing() -> (Dyn_Pickaxe_Swing:bool))
   > Set Dyn_Simulation_Space(Dyn_Simulation_Space:AnimPhysSimSpaceType Enum=BoneRelative) -> (Output_Get:AnimPhysSimSpaceType Enum)
   > Set Dyn Gravity Override 1(Dyn Gravity Override 1:Vector="-1900,0,0") -> (Output_Get:Vector)
   > Set Dyn Gravity Override 2(Dyn Gravity Override 2:Vector="-1250,0,0") -> (Output_Get:Vector)
   > Set Dyn Gravity Override 3(Dyn Gravity Override 3:Vector="-980,0,0") -> (Output_Get:Vector)
   Set Dyn_Simulation_Space(execute=@n3.else, Dyn_Simulation_Space=World)
   > Set Dyn Gravity Override 1(Dyn Gravity Override 1="0,0,-1900")
   > Set Dyn Gravity Override 2(Dyn Gravity Override 2="0,0,-1250")
   > Set Dyn Gravity Override 3(Dyn Gravity Override 3="0,0,-980")
   > Set Dyn_Tail_1_Alpha(Dyn_Tail_1_Alpha:float=0) -> (Output_Get:float)
   > Set Dyn_Tail_2_Alpha(Dyn_Tail_2_Alpha:float=0) -> (Output_Get:float)
   > Set Dyn_Tail_3_Alpha(Dyn_Tail_3_Alpha:float=0) -> (Output_Get:float)
   Set Dyn_Simulation_Space(execute=@n2.else, Dyn_Simulation_Space=BoneRelative)
   > Set Dyn Gravity Override 1(Dyn Gravity Override 1="-1900,0,0")
   > Set Dyn Gravity Override 2(Dyn Gravity Override 2="-1250,0,0")
   > Set Dyn Gravity Override 3(Dyn Gravity Override 3="-980,0,0")
   > Set Dyn_Tail_1_Alpha_FE(Dyn_Tail_1_Alpha_FE:float=0) -> (Output_Get:float)
   > Set Dyn_Tail_2_Alpha_FE(Dyn_Tail_2_Alpha_FE:float=0) -> (Output_Get:float)
   > Set Dyn_Tail_3_Alpha_FE(Dyn_Tail_3_Alpha_FE:float=0) -> (Output_Get:float)
   n4 = Branch(execute=@n1.then_1, Condition=Get bIsSliding())
   > Set Dyn_Tail_1_Alpha(Dyn_Tail_1_Alpha=1)
   Set Dyn_Tail_1_Alpha(execute=@n4.else, Dyn_Tail_1_Alpha=0)
   Event OnEndPickaxeSwing() {is=title K2Node_Event}
   > Set Dyn_Pickaxe_Swing(Dyn_Pickaxe_Swing:bool=false) -> (Output_Get:bool)
   Event OnBeginPickaxeSwing() {is=title K2Node_Event}
   > Set Dyn_Pickaxe_Swing(Dyn_Pickaxe_Swing=true)
   TryGetPawnOwner()
  ]}
]
variables [
 Dyn Gravity Override 1 {type=Vector, default=<-1900 0 0>}
 Dyn Gravity Override 2 {type=Vector, default=<-1250 0 0>}
 Dyn Gravity Override 3 {type=Vector, default=<-980 0 0>}
 Dyn_Pickaxe_Swing {type=boolean}
 Dyn_Simulation_Space {type=AnimPhysSimSpaceType, default=World}
 Dyn_Tail_1_Alpha {type=number, default=0.9}
 Dyn_Tail_1_Alpha_FE {type=number, default=0.9}
 Dyn_Tail_2_Alpha {type=number, default=0.85}
 Dyn_Tail_2_Alpha_FE {type=number, default=0.85}
 Dyn_Tail_3_Alpha {type=number, default=0.75}
 Dyn_Tail_3_Alpha_FE {type=number, default=0.75}
]
